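# EQS query that picks a destination for bringing ButterCake back toward its leash area.
# It generates points on a circle around the leash center context, with the radius bound to a float-curve-over-game-difficulty data provider.
# Points project onto navigation using the ButterCake nav filter and are lifted slightly afterward.
# A single random test filters and scores the points, so the chosen spot varies between runs.
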EnvQuery /ButterCake/EQS/EQS_BackToLeash
options [{
  generator {
   name=EnvQueryGenerator_OnCircle_5
   class=EnvQueryGenerator_OnCircle
   properties {
    circleCenter=&/EnvQueryContext_ButterCake_LeashCenter.EnvQueryContext_ButterCake_LeashCenter_C
    numberOfPoints=10
    circleRadius {
     dataBinding=&self:EnvQueryGenerator_OnCircle_5.FortAIDataProvider_FloatCurveOverGameDifficulty_0
     dataField=FloatValue
     defaultValue=5000
    }
    projectionData {
     extentX=10
     navigationFilter=/ButterCake/NavFilters/NavFilter_ButterCake.NavFilter_ButterCake_C
     postProjectionVerticalOffset=30
     projectDown=1024
    }
    traceData {navigationFilter=&NavFilter_ButterCake_C}
   }
  }
  tests [FortQueryTest_Random_4 {class=FortQueryTest_Random, properties {floatValueMax=1, scoringFactor=1, testPurpose=FilterAndScore}}]
 }]
