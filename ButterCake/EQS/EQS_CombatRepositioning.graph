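# EQS query that picks a combat repositioning point for the ButterCake AI.
# It generates a ring of navmesh-projected points using the ButterCake nav filter, with a radius driven by a game-difficulty curve.
# Distance tests drop points too close to the querier and filter and score against a difficulty-scaled leash around the leash center.
# A dot test compares querier-to-point and querier-to-goal directions to prefer having the goal in front.
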EnvQuery /ButterCake/EQS/EQS_CombatRepositioning
options [{
  generator {
   name=EnvQueryGenerator_OnCircle_5
   class=EnvQueryGenerator_OnCircle
   properties {
    numberOfPoints=10
    circleRadius {
     dataBinding=&self:EnvQueryGenerator_OnCircle_5.FortAIDataProvider_FloatCurveOverGameDifficulty_0
     dataField=FloatValue
     defaultValue=5000
    }
    projectionData {
     extentX=10
     navigationFilter=/ButterCake/NavFilters/NavFilter_ButterCake.NavFilter_ButterCake_C
     postProjectionVerticalOffset=30
     projectDown=1024
    }
    traceData {navigationFilter=&NavFilter_ButterCake_C}
   }
  }
  tests [
   EnvQueryTest_Distance_3 {
    class=EnvQueryTest_Distance
    properties {
     distanceTo=&EnvQueryContext_Querier
     filterType=Minimum
     floatValueMin=1500
     scoringEquation=Linear
     scoringFactor=1
     testComment=Remove points too close
     testMode=Distance3D
     testPurpose=Filter
    }
   }
   EnvQueryTest_Distance_11 {
    class=EnvQueryTest_Distance
    properties {
     distanceTo=&/EnvQueryContext_ButterCake_LeashCenter.EnvQueryContext_ButterCake_LeashCenter_C
     filterType=Maximum
     floatValueMin=0
     scoringEquation=Linear
     scoringFactor=1
     testComment=Location must be inside the Leash
     testMode=Distance2D
     testPurpose=FilterAndScore
     floatValueMax {
      dataBinding=&self:EnvQueryTest_Distance_11.FortAIDataProvider_FloatCurveOverGameDifficulty_0
      dataField=FloatValue
      defaultValue=0
     }
    }
   }
   EnvQueryTest_Dot_0 {
    class=EnvQueryTest_Dot
    properties {
     clampMaxType=SpecifiedValue
     clampMinType=SpecifiedValue
     testComment=Prefer Goal in front
     lineA {dirMode=TwoPoints, lineFrom=&EnvQueryContext_Querier, lineTo=&EnvQueryContext_Item}
     lineB {lineTo=&FortQueryContext_Goal, rotation=&EnvQueryContext_Querier}
     scoreClampMax {dataBinding=null, dataField=FloatValue, defaultValue=1}
     scoreClampMin {dataBinding=null, dataField=FloatValue, defaultValue=-1}
     scoringFactor {dataBinding=null, dataField=FloatValue, defaultValue=1}
    }
   }
  ]
 }]
