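# EQS query that picks a goal for the ButterCake NPC among perceived enemies, loot and trackable AI objects, via a composite generator.
# Filters drop loot outside a difficulty-scaled leash or too far away, world goals off the ButterCake nav filter, and frustrated goals.
# Scoring favors big fruit items and plants, freshly player-tossed pickups, closer goals, goals in front and sensed targets, with hearing weighted.
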EnvQuery /ButterCake/EQS/EQS_GoalSelection_Pawn_ButterCake
options [{
  generator {
   name=EnvQueryGenerator_Composite_1
   class=EnvQueryGenerator_Composite
   properties {
    optionName=Perceived Enemies and Loots
    generators [
     &self:EnvQueryGenerator_Composite_1.FortQueryGenerator_Enemies_1
     &self:EnvQueryGenerator_Composite_1.FortQueryGenerator_LootGoalsAthena_0
     &self:EnvQueryGenerator_Composite_1.FortQueryGenerator_GoalTrackableAIObjects_0
    ]
   }
  }
  tests [
   FortQueryTest_GoalDistance_0 {
    class=FortQueryTest_GoalDistance
    properties {
     bScoreEncounterGoals=false
     bScoreEnemies=false
     bScoreSpecificAssignments=true
     bScoreWorldGoals=true
     clampMaxType=None
     clampMinType=None
     distanceTo=&/EnvQueryContext_ButterCake_LeashCenter.EnvQueryContext_ButterCake_LeashCenter_C
     filterType=Maximum
     scoreClampMax=0
     scoreClampMin=0
     scoringEquation=Linear
     scoringFactor=1
     testComment=Leash : Remove loot outside the leash
     testMode=Distance2D
     testPurpose=Filter
     assignmentIds [{assignmentTeam=Monster, assignmentType=World, assignmentGameplayTags {gameplayTags [{tagName=NPC.Assignment.Loot}]}}]
     floatValueMax {
      dataBinding=&self:FortQueryTest_GoalDistance_0.FortAIDataProvider_FloatCurveOverGameDifficulty_0
      dataField=FloatValue
      defaultValue=0
     }
    }
   }
   FortQueryTest_GoalGameplayTags_0 {
    class=FortQueryTest_GoalGameplayTags
    properties {
     bDefineReferenceValue=false
     bRequireAllProvidedTagQueriesPass=false
     bScoreEncounterGoals=false
     bScoreEnemies=false
     bScoreSpecificAssignments=true
     bScoreWorldGoals=false
     bShouldLookupQueryByTag=false
     bShouldPassWhenQueryNotFound=true
     boolValue=true
     clampMaxType=None
     clampMinType=None
     filterType=Match
     floatValueMax=0
     floatValueMin=0
     multipleContextFilterOp=AllPass
     multipleContextScoreOp=AverageScore
     normalizationType=Absolute
     referenceValue=0
     scoreClampMax=0
     scoreClampMin=0
     scoringEquation=Constant
     scoringFactor=4
     testComment=Prioritize Fruit Item
     testPurpose=Score
     assignmentIds [{assignmentTeam=Monster, assignmentType=World, assignmentGameplayTags {gameplayTags [{tagName=NPC.Assignment.Loot}]}}]
     goalActorTagQuery {autoDescription="", tokenStreamVersion=0, userDescription="", queryTokenStream [], tagDictionary []}
     pawnOwnerTagQuery {autoDescription="", tokenStreamVersion=0, userDescription="", queryTokenStream [], tagDictionary []}
     prohibitedAssignmentIds []
     queryLookupTag {tagName=Item.Consumable.ShieldPotion}
     tagQueryToMatch {
      autoDescription=" ANY( Item.Consumable.BigFruit )"
      tokenStreamVersion=0
      userDescription=""
      queryTokenStream [0,1,1,1,0]
      tagDictionary [{tagName=Item.Consumable.BigFruit}]
     }
    }
   }
   EnvQueryTest_GameplayTags_0 {
    class=EnvQueryTest_GameplayTags
    properties {
     bDefineReferenceValue=false
     bRejectIncompatibleItems=false
     bSearchComponentsForTags=false
     boolValue=true
     clampMaxType=None
     clampMinType=None
     filterType=Match
     floatValueMax=0
     floatValueMin=0
     multipleContextFilterOp=AllPass
     multipleContextScoreOp=AverageScore
     normalizationType=Absolute
     referenceValue=0
     scoreClampMax=0
     scoreClampMin=0
     scoringEquation=Constant
     scoringFactor=2
     testComment=""
     testPurpose=Score
     tagQueryToMatch {
      autoDescription=" ANY( Building.Type.Prop.BigFruitPlant )"
      tokenStreamVersion=0
      userDescription=""
      queryTokenStream [0,1,1,1,0]
      tagDictionary [{tagName=Building.Type.Prop.BigFruitPlant}]
     }
    }
   }
   FortQueryTest_GoalPickupFilter_1 {
    class=FortQueryTest_GoalPickupFilter
    properties {
     bDefineReferenceValue=false
     bScoreEncounterGoals=false
     bScoreEnemies=false
     bScoreSpecificAssignments=true
     bScoreWorldGoals=false
     boolValue=true
     clampMaxType=None
     clampMinType=None
     filterType=Match
     floatValueMax=0
     floatValueMin=0
     maxLifeTime=10
     multipleContextFilterOp=AllPass
     multipleContextScoreOp=AverageScore
     normalizationType=Absolute
     referenceValue=0
     requiredPickupSpawnSource=TossedByPlayer
     scoreClampMax=0
     scoreClampMin=0
     scoringEquation=Constant
     scoringFactor=1
     testComment="Prioritize Fresh Items (recently tossed by players)"
     testPurpose=Score
     assignmentIds [{assignmentTeam=Monster, assignmentType=World, assignmentGameplayTags {gameplayTags [{tagName=NPC.Assignment.Loot}]}}]
     goalActorTagQuery {autoDescription="", tokenStreamVersion=0, userDescription="", queryTokenStream [], tagDictionary []}
     prohibitedAssignmentIds []
    }
   }
   FortQueryTest_GoalDistance_3 {
    class=FortQueryTest_GoalDistance
    properties {
     bScoreEncounterGoals=false
     bScoreEnemies=false
     bScoreSpecificAssignments=true
     bScoreWorldGoals=false
     clampMaxType=SpecifiedValue
     clampMinType=SpecifiedValue
     filterType=Maximum
     floatValueMax=125000
     scoreClampMax=5000
     scoreClampMin=200
     scoringEquation=InverseLinear
     scoringFactor=1
     testComment=Loot Scoring and Filtering
     testPurpose=FilterAndScore
     assignmentIds [{assignmentTeam=Monster, assignmentType=World, assignmentGameplayTags {gameplayTags [{tagName=NPC.Assignment.Loot}]}}]
    }
   }
   FortQueryTest_GoalDistance_4 {
    class=FortQueryTest_GoalDistance
    properties {
     bScoreEncounterGoals=true
     bScoreEnemies=true
     bScoreSpecificAssignments=false
     bScoreWorldGoals=false
     clampMaxType=SpecifiedValue
     clampMinType=SpecifiedValue
     filterType=Range
     floatValueMax=0
     scoreClampMax=10000
     scoreClampMin=3000
     scoringEquation=InverseLinear
     scoringFactor=1
     testComment=Prefer closer things.
     testPurpose=Score
     assignmentIds []
    }
   }
   FortQueryTest_PerceptionAll_1 {
    class=FortQueryTest_PerceptionAll
    properties {
     clampMaxType=SpecifiedValue
     clampMinType=SpecifiedValue
     maxSenseAge=21
     minSenseAge=20
     scoreClampMax=1
     scoringFactor=3
     additionalSenseScores {&Class'&AthenaAISense_Hearing'=1}
     senseScores [0,0,0,0,0,0]
    }
   }
   FortQueryTest_GoalProject_0 {
    class=FortQueryTest_GoalProject
    properties {
     bDefineReferenceValue=false
     bScoreEncounterGoals=false
     bScoreEnemies=false
     bScoreSpecificAssignments=false
     bScoreWorldGoals=true
     boolValue=true
     clampMaxType=None
     clampMinType=None
     filterType=Match
     floatValueMax=0
     floatValueMin=0
     multipleContextFilterOp=AllPass
     multipleContextScoreOp=AverageScore
     normalizationType=Absolute
     referenceValue=0
     scoreClampMax=0
     scoreClampMin=0
     scoringEquation=Constant
     scoringFactor=1
     testComment="Only World (Loot) Goals near Nav"
     testPurpose=Filter
     assignmentIds []
     goalActorTagQuery {autoDescription="", tokenStreamVersion=0, userDescription="", queryTokenStream [], tagDictionary []}
     prohibitedAssignmentIds []
     projectionData {
      bCanDisableTrace=false
      bCanProjectDown=true
      bCanTraceOnGeometry=false
      bCanTraceOnNavMesh=true
      bOnlyBlockingHits=true
      bTraceComplex=false
      extentX=0
      extentY=10
      extentZ=10
      navigationFilter=/ButterCake/NavFilters/NavFilter_ButterCake.NavFilter_ButterCake_C
      postProjectionVerticalOffset=0
      projectDown=512
      projectUp=512
      serializedChannel=ECC_Visibility
      traceChannel=TraceTypeQuery1
      traceMode=Navigation
      traceProfileName=None
      traceShape=Line
      versionNum=1
     }
    }
   }
   FortQueryTest_GoalFrustrationDiscouragement_0 {
    class=FortQueryTest_GoalFrustrationDiscouragement
    properties {
     bDefineReferenceValue=false
     bScoreEncounterGoals=true
     bScoreEnemies=true
     bScoreSpecificAssignments=false
     bScoreWorldGoals=true
     boolValue=true
     clampMaxType=SpecifiedValue
     clampMinType=SpecifiedValue
     filterType=Maximum
     floatValueMax=1
     floatValueMin=0
     multipleContextFilterOp=AllPass
     multipleContextScoreOp=AverageScore
     normalizationType=Absolute
     referenceValue=0
     scoreClampMax=1
     scoreClampMin=0
     scoringEquation=Constant
     scoringFactor=-1
     testComment="When a Goal can't be reached, we can choose to Frustrate on (IGNORE) that Goal for a period of time."
     testPurpose=FilterAndScore
     assignmentIds []
     goalActorTagQuery {autoDescription="", tokenStreamVersion=0, userDescription="", queryTokenStream [], tagDictionary []}
     prohibitedAssignmentIds []
    }
   }
   EnvQueryTest_Dot_5 {
    class=EnvQueryTest_Dot
    properties {
     clampMaxType=SpecifiedValue
     clampMinType=SpecifiedValue
     testComment=Prefer Goal in front
     lineA {dirMode=TwoPoints, lineFrom=&EnvQueryContext_Querier, lineTo=&EnvQueryContext_Item}
     lineB {dirMode=Rotation, lineTo=&FortQueryContext_Goal, rotation=&EnvQueryContext_Querier}
     scoreClampMax {dataBinding=null, dataField=FloatValue, defaultValue=1}
     scoreClampMin {dataBinding=null, dataField=FloatValue, defaultValue=-1}
     scoringFactor {dataBinding=null, dataField=FloatValue, defaultValue=1}
    }
   }
  ]
 }]
