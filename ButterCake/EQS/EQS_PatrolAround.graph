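# EQS query that picks a patrol point for the ButterCake AI within its leash area.
# The primary option samples a navmesh-projected ring with difficulty-scaled radius, with a dot test against the querier's facing, though commented as preferring the goal direction.
# A fallback grid option omits that directional test.
# Both reject points too near the querier, filter and score by difficulty-scaled 2D distance from the leash center, and add randomness.
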EnvQuery /ButterCake/EQS/EQS_PatrolAround
options [
 {
  generator {
   name=EnvQueryGenerator_OnCircle_5
   class=EnvQueryGenerator_OnCircle
   properties {
    numberOfPoints=10
    circleRadius {
     dataBinding=&self:EnvQueryGenerator_OnCircle_5.FortAIDataProvider_FloatCurveOverGameDifficulty_0
     dataField=FloatValue
     defaultValue=5000
    }
    projectionData {
     extentX=10
     navigationFilter=/ButterCake/NavFilters/NavFilter_ButterCake.NavFilter_ButterCake_C
     postProjectionVerticalOffset=30
     projectDown=5000
     projectUp=5000
    }
    traceData {navigationFilter=&NavFilter_ButterCake_C}
   }
  }
  tests [
   EnvQueryTest_Distance_3 {
    class=EnvQueryTest_Distance
    properties {
     distanceTo=&EnvQueryContext_Querier
     filterType=Minimum
     floatValueMin=1500
     scoringEquation=Linear
     scoringFactor=1
     testComment=""
     testMode=Distance3D
     testPurpose=Filter
    }
   }
   EnvQueryTest_Distance_11 {
    class=EnvQueryTest_Distance
    properties {
     distanceTo=&/EnvQueryContext_ButterCake_LeashCenter.EnvQueryContext_ButterCake_LeashCenter_C
     filterType=Maximum
     floatValueMin=0
     scoringEquation=Linear
     scoringFactor=1
     testComment=""
     testMode=Distance2D
     testPurpose=FilterAndScore
     floatValueMax {
      dataBinding=&self:EnvQueryTest_Distance_11.FortAIDataProvider_FloatCurveOverGameDifficulty_0
      dataField=FloatValue
      defaultValue=0
     }
    }
   }
   EnvQueryTest_Dot_5 {
    class=EnvQueryTest_Dot
    properties {
     clampMaxType=SpecifiedValue
     clampMinType=SpecifiedValue
     scoringEquation=Constant
     testComment=Prefer in direction of Goal
     lineA {dirMode=TwoPoints, lineFrom=&EnvQueryContext_Querier, lineTo=&EnvQueryContext_Item}
     lineB {dirMode=Rotation, lineTo=&FortQueryContext_Goal, rotation=&EnvQueryContext_Querier}
     scoreClampMax {dataBinding=null, dataField=FloatValue, defaultValue=1}
     scoreClampMin {dataBinding=null, dataField=FloatValue, defaultValue=0.5}
     scoringFactor {dataBinding=null, dataField=FloatValue, defaultValue=1}
    }
   }
   FortQueryTest_Random_4 {class=FortQueryTest_Random, properties {floatValueMax=1, scoringFactor=1, testPurpose=FilterAndScore}}
  ]
 }
 {
  generator {
   name=EnvQueryGenerator_SimpleGrid_0
   class=EnvQueryGenerator_SimpleGrid
   properties {
    gridSize=5000
    spaceBetween=2000
    projectionData {
     extentX=10
     navigationFilter=&NavFilter_ButterCake_C
     postProjectionVerticalOffset=30
     projectDown=5000
     projectUp=5000
    }
   }
  }
  tests [
   EnvQueryTest_Distance_1 {
    class=EnvQueryTest_Distance
    properties {
     distanceTo=&EnvQueryContext_Querier
     filterType=Minimum
     floatValueMin=1500
     scoringEquation=Linear
     scoringFactor=1
     testComment=""
     testMode=Distance3D
     testPurpose=Filter
    }
   }
   EnvQueryTest_Distance_0 {
    class=EnvQueryTest_Distance
    properties {
     distanceTo=&EnvQueryContext_ButterCake_LeashCenter_C
     filterType=Maximum
     floatValueMin=0
     scoringEquation=Linear
     scoringFactor=1
     testComment=""
     testMode=Distance2D
     testPurpose=FilterAndScore
     floatValueMax {
      dataBinding=&self:EnvQueryTest_Distance_0.FortAIDataProvider_FloatCurveOverGameDifficulty_0
      dataField=FloatValue
      defaultValue=0
     }
    }
   }
   FortQueryTest_Random_0 {class=FortQueryTest_Random, properties {floatValueMax=1, scoringFactor=1, testPurpose=FilterAndScore}}
  ]
 }
]
