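; EQS context for ButterCake AI that supplies the querier's leash center as a single location for environment queries.
; It casts the querier to a Fort AI pawn, gets its AI controller and reads the center from that controller's FortAthenaLeashComponent.
; If the controller or leash component is invalid, it falls back to the pawn's own location.
; A failed cast supplies no explicit location.
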(fn ProvideSingleLocation
  (params (QuerierObject :type Object%20Reference) (QuerierActor :type Actor%20Object%20Reference))
  (bind _asfort_aipawn (Utilities|Casting|CastToFortAIPawn QuerierActor)
    (:then
      (Utilities|IsValid (AI|GetAIController _asfort_aipawn)
        (:Is%20Valid
          (Utilities|IsValid (Actor|GetComponentbyClass (AI|GetAIController _asfort_aipawn) "/Script/FortniteAI.FortAthenaLeashComponent")
            (:Is%20Valid
              (return (Class|FortAthenaLeashComponent|GetLeashCenter (Actor|GetComponentbyClass (AI|GetAIController _asfort_aipawn) "/Script/FortniteAI.FortAthenaLeashComponent"))))
            (:Is%20Not%20Valid
              _)))
        (:Is%20Not%20Valid
          _))
      (return (Transformation|GetActorLocation _asfort_aipawn)))
    (:CastFailed
      _)))
